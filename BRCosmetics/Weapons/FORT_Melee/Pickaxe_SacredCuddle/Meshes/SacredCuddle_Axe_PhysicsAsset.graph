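# Physics asset for the SacredCuddle axe, a cosmetic pickaxe in the melee weapon set.
# It has a single body on the Melee_TwoHanded_Handle bone, made of one offset, rotated capsule.
# With no other bodies and no constraints, it gives the weapon one simple collision shape and no jointed ragdoll parts.
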PhysicsAsset /BRCosmetics/Weapons/FORT_Melee/Pickaxe_SacredCuddle/Meshes/SacredCuddle_Axe_PhysicsAsset
bodies {Melee_TwoHanded_Handle [{shapeType=Capsule, length=39.3123, radius=6.0828, center=<-0.0009 -3.736 19.2261>, rotation=<rot 0.7618 -90 -0.0047>}]}
constraints []
