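# Physics asset for the left-hand axe of the Sacred Cuddle melee pickaxe cosmetic skeletal mesh.
# It holds a single body on the b_AR_Root bone, shaped as one offset, rotated capsule that approximates the axe.
# There are no constraints, so the asset defines only one collision shape rather than a jointed ragdoll.
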PhysicsAsset /BRCosmetics/Weapons/FORT_Melee/Pickaxe_SacredCuddle/Meshes/SacredCuddle_L_Axe_PhysicsAsset
bodies {b_AR_Root [{shapeType=Capsule, length=38.9574, radius=6.4296, center=<-2.5455 -0.0036 18.6946>, rotation=<rot -0.0227 -90 3.3957>}]}
constraints []
